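# Physics asset for the CondorMotor vehicle body skeletal mesh, in the VehicleCosmetics bodies content.
# It holds a single body on the root bone, shaped as one box offset forward and upward from the bone origin.
# It defines no constraints, so the whole vehicle is represented as one simplified collision volume with no articulated parts.
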PhysicsAsset /VehicleCosmetics/Bodies/CondorMotor/Meshes/SK_CondorMotor_PhysicsDM
bodies {root [{shapeType=Box, extentX=500, extentY=208, extentZ=136, shapeName=root_box, center=<31 0 104>}]}
constraints []
